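# Physics asset for the F_MED_CameraShake female medium player body, set up for two dangling drawstrings.
# Each string is a capsule chain from its MSTR bone, with four segments on the left and three on the right, joined by swing- and twist-limited constraints.
# Larger pelvis and thigh capsules have no constraints, apparently serving as collision volumes for the strings.
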PhysicsAsset /Game/Characters/Player/Female/Medium/Bodies/F_MED_CameraShake/Meshes/F_MED_CameraShake_Physics_Base
bodies {
 dyn_draw_string_L_01 [{shapeType=Capsule, length=2.5, radius=1, shapeName=dyn_draw_string_L_01_capsule, center=<1.249 0 0>, rotation=<rot 90 -166.7373 -166.7373>}]
 dyn_draw_string_L_02 [{shapeType=Capsule, length=2.5, radius=1, shapeName=dyn_draw_string_L_02_capsule, center=<1.249 0 0>, rotation=<rot 90 90 90>}]
 dyn_draw_string_L_03 [{shapeType=Capsule, length=2, radius=1.5, shapeName=dyn_draw_string_L_03_capsule, center=<1.249 0 0>, rotation=<rot 90 0.0008 0.0008>}]
 dyn_draw_string_L_04 [{shapeType=Capsule, length=2, radius=1.5, shapeName=dyn_draw_string_L_04_capsule, center=<1.249 0 0>, rotation=<rot 90 90 90>}]
 dyn_draw_string_L_MSTR [{shapeType=Capsule, length=1.5, radius=1, shapeName=dyn_draw_string_L_MSTR_capsule, rotation=<rot 0.0001 0 -90>}]
 dyn_draw_string_R_01 [{shapeType=Capsule, length=2.5, radius=1, shapeName=dyn_draw_string_R_01_capsule, center=<1.249 0 0>, rotation=<rot 90 -180 180>}]
 dyn_draw_string_R_02 [{shapeType=Capsule, length=2.5, radius=1, shapeName=dyn_draw_string_R_02_capsule, center=<1.249 0 0>, rotation=<rot 90 0 0>}]
 dyn_draw_string_R_03 [{shapeType=Capsule, length=2.5, radius=1, shapeName=dyn_draw_string_R_03_capsule, center=<1.249 0 0>, rotation=<rot 90 0 0>}]
 dyn_draw_string_R_MSTR [{shapeType=Capsule, length=1.5, radius=1, shapeName=dyn_draw_string_R_MSTR_capsule, rotation=<rot 0.0001 0 -90>}]
 pelvis [{shapeType=Capsule, length=10, radius=17, shapeName=pelvis_capsule, center=<-8.2264 -6.4775 0>, rotation=<rot 90 0 6.9313>}]
 thigh_l [{shapeType=Capsule, length=10, radius=8, shapeName=thigh_l_capsule, center=<-9.3306 0.0934 3.1485>, rotation=<rot -21.4979 -0.0002 -13.0552>}]
 thigh_r [{shapeType=Capsule, length=9.5, radius=8, shapeName=thigh_r_capsule, center=<8.7083 -1.1118 -4.1336>, rotation=<rot -19.1594 3.035 -19.0911>}]
}
constraints [
 {bone1Name=dyn_draw_string_L_01, bone2Name=dyn_draw_string_L_MSTR, swing1LimitDegrees=20, swing2LimitDegrees=20, twistLimitDegrees=10}
 {bone1Name=dyn_draw_string_R_01, bone2Name=dyn_draw_string_R_MSTR, swing1LimitDegrees=20, swing2LimitDegrees=20, twistLimitDegrees=10}
 {bone1Name=dyn_draw_string_L_02, bone2Name=dyn_draw_string_L_01, swing1LimitDegrees=20, swing2LimitDegrees=20, twistLimitDegrees=10}
 {bone1Name=dyn_draw_string_L_03, bone2Name=dyn_draw_string_L_02, swing1LimitDegrees=20, swing2LimitDegrees=30, twistLimitDegrees=10}
 {bone1Name=dyn_draw_string_L_04, bone2Name=dyn_draw_string_L_03, swing1LimitDegrees=20, swing2LimitDegrees=30, twistLimitDegrees=10}
 {bone1Name=dyn_draw_string_R_02, bone2Name=dyn_draw_string_R_01, swing1LimitDegrees=20, swing2LimitDegrees=20, twistLimitDegrees=10}
 {bone1Name=dyn_draw_string_R_03, bone2Name=dyn_draw_string_R_02, swing1LimitDegrees=20, swing2LimitDegrees=30, twistLimitDegrees=10}
]
