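# Post-process style AnimBP for the female CameraShake player body, adding secondary motion to necklace, straps, buckle, loops and draw strings.
# It copies the parent pose through a deform Control Rig, then caches an AnimDynamics chain.
# RigidBody, with crouch-specific physics, refines this when enabled; otherwise thigh constraints and crouch bone offsets pose the draw strings.
# The EventGraph computes right-thigh alpha and feeds foot, calf and skydive wind data to WPO cloth materials.
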AnimBlueprint /Game/Characters/Player/Female/Medium/Bodies/F_MED_CameraShake/Meshes/F_MED_CameraShake_AnimBP
target_skeleton=&/F_MED_CameraShake_Skeleton
graphs [
 AnimGraph {nodes [
   n1 = "Transform (Modify) Bone - Bone: dyn_draw_string_R_MSTR"!(
    ComponentPose=Local To Component(LocalPose=Use cached pose 'AnimDynPose'())
    bAlphaBoolEnabled:bool=True
    Translation="0,0,0"
    Rotation="-15,0,15"
    Scale="1,1,1"
   ) -> (Pose) {
    alphaInputType=Bool
    rotationMode=BMM_Additive
    alphaBoolBlend {blendOption=Linear}
    boneToModify {boneName=dyn_draw_string_R_MSTR}
    rotation {pitch=-15, roll=15}
   }
   n2 = "Transform (Modify) Bone - Bone: dyn_draw_string_L_MSTR"!(ComponentPose=@n1, bAlphaBoolEnabled:bool=NOT Boolean(A=Get bIsMoving2D()), Translation="0,0,0", Rotation="-10,0,70", Scale="1,1,1") -> (Pose) {
    alphaInputType=Bool
    rotationMode=BMM_Additive
    alphaBoolBlend {blendOption=Linear}
    boneToModify {boneName=dyn_draw_string_L_MSTR}
    rotation {pitch=-10, roll=70}
   }
   n3 = "Transform (Modify) Bone - Bone: dyn_draw_string_L_MSTR"(ComponentPose=@n2, bAlphaBoolEnabled=Get bIsMoving2D(), Translation="0,0,0", Rotation="-20,0,-40", Scale="1,1,1") {
    alphaInputType=Bool
    rotationMode=BMM_Additive
    alphaBoolBlend {blendOption=Linear}
    boneToModify {boneName=dyn_draw_string_L_MSTR}
    rotation {pitch=-20, roll=-40}
   }
   n4 = Component To Local(ComponentPose=@n3)
   n5 = Use cached pose 'AnimDynPose'()
   n6 = Get bIsCrouching()
   n7 = "Transform (Modify) Bone - Bone: dyn_draw_string_R_MSTR"(
    ComponentPose=Local To Component(LocalPose=Use cached pose 'AnimDynPose'())
    bAlphaBoolEnabled=True
    Translation="0,0,0"
    Rotation="-30,0,55"
    Scale="1,1,1"
   ) {
    alphaInputType=Bool
    rotationMode=BMM_Additive
    alphaBoolBlend {blendOption=Linear}
    boneToModify {boneName=dyn_draw_string_R_MSTR}
    rotation {pitch=-30, roll=55}
   }
   n8 = "Transform (Modify) Bone - Bone: dyn_draw_string_L_MSTR"(ComponentPose=@n7, bAlphaBoolEnabled=True, Translation="0,0,0", Rotation="0,0,55", Scale="1,1,1") {
    alphaInputType=Bool
    rotationMode=BMM_Additive
    alphaBoolBlend {blendOption=Linear}
    boneToModify {boneName=dyn_draw_string_L_MSTR}
    rotation {roll=55}
   }
   n9 = Component To Local(ComponentPose=@n8)
   n10 = Use cached pose 'AnimDynPose'()
   n11 = Get bIsMoving2D()
   n12 = Get AlphaThighR() -> (AlphaThighR:double)
   n13 = "Transform (Modify) Bone - Bone: dyn_draw_string_R_MSTR"!(
    ComponentPose=Local To Component(LocalPose=Blend Poses by bool(BlendPose_0=@n10, BlendPose_1=@n9, BlendTime_0=1, BlendTime_1=1, bActiveValue=@n11) {blendType=Linear})
    Alpha:float=@n12
    Translation="0,0,0"
    Rotation="0,0,-40"
    Scale="1,1,1"
   ) -> (Pose) {rotationMode=BMM_Additive, alphaBoolBlend {blendOption=Linear}, boneToModify {boneName=dyn_draw_string_R_MSTR}, rotation {roll=-40}}
   n14 = "Transform (Modify) Bone - Bone: dyn_draw_string_L_MSTR"!(ComponentPose=@n13, Alpha:float=@n12, Translation="0,0,0", Rotation="0,0,-40", Scale="1,1,1") -> (Pose) {rotationMode=BMM_Additive, alphaBoolBlend {blendOption=Linear}, boneToModify {boneName=dyn_draw_string_L_MSTR}, rotation {roll=-40}}
   n15 = RigidBody(ComponentPose=@n14) {
    overridePhysicsAsset=&/F_MED_CameraShake_Physics_Crouching
    simulationSpace=BaseBoneSpace
    alphaBoolBlend {blendOption=Linear}
    baseBoneRef {boneName=spine_01}
    componentAppliedLinearAccClamp=<800 800 500>
    componentLinearAccScale=<0.01 0.01 0.01>
    componentLinearVelScale=<0.4 0.4 0.4>
   }
   n16 = Component To Local(ComponentPose=@n15)
   n17 = RigidBody(ComponentPose=Local To Component(LocalPose=Use cached pose 'AnimDynPose'())) {
    overridePhysicsAsset=&/F_MED_CameraShake_Physics_Base
    simulationSpace=BaseBoneSpace
    alphaBoolBlend {blendOption=Linear}
    baseBoneRef {boneName=spine_01}
    componentAppliedLinearAccClamp=<800 800 500>
    componentLinearAccScale=<0.01 0.01 0.01>
    componentLinearVelScale=<0.4 0.4 0.4>
   }
   n18 = Component To Local(ComponentPose=@n17)
   n19 = Get bIsCrouching()
   Output Pose(Result=Blend Poses by bool(
     BlendPose_0=Blend Poses by bool(BlendPose_0=@n16, BlendPose_1=@n18, bActiveValue=@n19) {blendType=Linear}
     BlendPose_1=Blend Poses by bool(BlendPose_0=@n4, BlendPose_1=@n5, bActiveValue=@n6) {blendType=Linear}
     bActiveValue=Get bRigidBodyEnabled()
    ) {blendType=Linear}) {layerGroup=None}
   n20 = Control Rig(Source=Copy Pose From Mesh() {bUseAttachedParent=true}, Alpha=0.8) {
    alphaBoolBlend {blendOption=Linear}
    controlRigAssetReference {blueprintRigClass=/Game/Characters/Player/Male/Medium/Bodies/M_MED_Boredom/Meshes/M_MED_DeformRig_Boredom_ControlRig.M_MED_DeformRig_Boredom_ControlRig_C}
   }
   n21 = Anim Dynamics - Bone: dyn_necklace_base(ComponentPose=Local To Component(LocalPose=@n20), Alpha=1, ExternalForce="0,0,-300") {
    angularDampingOverride=0.999
    angularSpringConstant=100
    bAngularSpring=true
    bOverrideAngularDamping=true
    bOverrideLinearDamping=true
    bUseGravityOverride=true
    gravityScale=-1
    linearDampingOverride=0.999
    numSolverIterationsPostUpdate=2
    numSolverIterationsPreUpdate=5
    simulationSpace=RootRelative
    alphaBoolBlend {blendOption=Linear}
    boundBone {boneName=dyn_necklace_base}
    componentAppliedLinearAccClamp=<5000 5000 5000>
    componentLinearAccScale=<0.2 0.2 0.2>
    componentLinearVelScale=<0.3 0.3 0.3>
    gravityOverride {z=-600}
    physicsBodyDefinitions [{
      boundBone {boneName=dyn_necklace_base}
      boxExtents=<12 16 8>
      constraintSetup {angularLimitsMax=<0 -0.5 5>, angularLimitsMin=<0 -60 -15>, angularTarget=<0 0 0>}
      localJointOffset=<3 0 0>
     }]
    relativeSpaceBone {boneName=neck_01}
   }
   n22 = Anim Dynamics - Bone: dyn_necklace_L(ComponentPose=@n21, Alpha=1, ExternalForce="0,0,-100") {
    angularDampingOverride=0.7
    angularSpringConstant=500
    bAngularSpring=true
    bOverrideAngularDamping=true
    bUseGravityOverride=true
    gravityScale=-1
    simulationSpace=RootRelative
    alphaBoolBlend {blendOption=Linear}
    boundBone {boneName=dyn_necklace_L}
    componentAppliedLinearAccClamp=<3000 3000 3000>
    componentLinearAccScale=<0.3 0.3 0.3>
    componentLinearVelScale=<0.3 0.3 0.3>
    gravityOverride {z=-600}
    physicsBodyDefinitions [{
      boundBone {boneName=dyn_necklace_L}
      boxExtents=<12 16 8>
      constraintSetup {angularLimitsMax=<5 5 5>, angularLimitsMin=<-5 -30 -5>, angularTarget=<0 0 0>}
      localJointOffset=<3 0 0>
     }]
   }
   n23 = Anim Dynamics - Bone: dyn_necklace_R(ComponentPose=@n22, Alpha=1, ExternalForce="0,0,-100") {
    angularDampingOverride=0.7
    angularSpringConstant=500
    bAngularSpring=true
    bOverrideAngularDamping=true
    bUseGravityOverride=true
    gravityScale=-1
    simulationSpace=RootRelative
    alphaBoolBlend {blendOption=Linear}
    boundBone {boneName=dyn_necklace_R}
    componentAppliedLinearAccClamp=<3000 3000 3000>
    componentLinearAccScale=<0.3 0.3 0.3>
    componentLinearVelScale=<0.3 0.3 0.3>
    gravityOverride {z=-600}
    physicsBodyDefinitions [{
      boundBone {boneName=dyn_necklace_R}
      boxExtents=<12 16 8>
      constraintSetup {angularLimitsMax=<5 5 5>, angularLimitsMin=<-5 -30 -5>, angularTarget=<0 0 0>}
      localJointOffset=<3 0 0>
     }]
   }
   n24 = Anim Dynamics - Bone: dyn_strap_R_01!(ComponentPose=@n23, Alpha=1, GravityOverride:Vector="0,0,-900") -> (Pose) {
    angularDampingOverride=0.9999
    angularSpringConstant=1000
    bAngularSpring=true
    bGravityOverrideInSimSpace=true
    bOverrideAngularDamping=true
    simulationSpace=RootRelative
    alphaBoolBlend {blendOption=Linear}
    boundBone {boneName=dyn_strap_R_01}
    componentAppliedLinearAccClamp=<5000 5000 5000>
    componentLinearAccScale=<0.5 0.5 0.5>
    componentLinearVelScale=<0.6 0.6 0.6>
    gravityOverride {z=-980}
    physicsBodyDefinitions [{
      sphereCollisionRadius=0
      boundBone {boneName=dyn_strap_R_01}
      boxExtents=<5 5 5>
      constraintSetup {angularLimitsMax=<0 15 0>, angularLimitsMin=<0 -60 0>, angularTarget=<0 0 0>}
      localJointOffset=<1 0 0>
     }]
    relativeSpaceBone {boneName=head}
   }
   n25 = Anim Dynamics - Bone: dyn_strap_R_02(ComponentPose=@n24, Alpha=1, ExternalForce="0,0,-300") {
    angularDampingOverride=0.999
    angularSpringConstant=500
    bAngularSpring=true
    bOverrideAngularDamping=true
    bUseGravityOverride=true
    gravityScale=-1
    simulationSpace=RootRelative
    alphaBoolBlend {blendOption=Linear}
    boundBone {boneName=dyn_strap_R_02}
    componentAppliedLinearAccClamp=<3000 3000 3000>
    componentLinearAccScale=<0.3 0.3 0.3>
    componentLinearVelScale=<0.3 0.3 0.3>
    gravityOverride {z=-980}
    physicsBodyDefinitions [{
      boundBone {boneName=dyn_strap_R_02}
      boxExtents=<12 16 8>
      constraintSetup {angularLimitsMax=<0 0 10>, angularLimitsMin=<0 -60 -10>, angularTarget=<0 0 0>}
      localJointOffset=<3 0 0>
     }]
   }
   n26 = Anim Dynamics - Bone: dyn_strap_R_03(ComponentPose=@n25, Alpha=1, ExternalForce="0,0,-900") {
    angularDampingOverride=0.999
    angularSpringConstant=500
    bAngularSpring=true
    bOverrideAngularDamping=true
    bUseGravityOverride=true
    gravityScale=-1
    simulationSpace=RootRelative
    alphaBoolBlend {blendOption=Linear}
    boundBone {boneName=dyn_strap_R_03}
    componentAppliedLinearAccClamp=<3000 3000 3000>
    componentLinearAccScale=<0.3 0.3 0.3>
    componentLinearVelScale=<0.3 0.3 0.3>
    gravityOverride {z=-980}
    physicsBodyDefinitions [{
      boundBone {boneName=dyn_strap_R_03}
      boxExtents=<12 16 8>
      constraintSetup {angularLimitsMax=<10 0 10>, angularLimitsMin=<-10 -40 -10>, angularTarget=<0 0 0>}
      localJointOffset=<3 0 0>
     }]
   }
   n27 = Anim Dynamics - Bone: dyn_strap_R_04(ComponentPose=@n26, Alpha=1, ExternalForce="0,0,-400") {
    angularDampingOverride=0.999
    angularSpringConstant=500
    bAngularSpring=true
    bOverrideAngularDamping=true
    bUseGravityOverride=true
    gravityScale=-1
    simulationSpace=RootRelative
    alphaBoolBlend {blendOption=Linear}
    boundBone {boneName=dyn_strap_R_04}
    componentAppliedLinearAccClamp=<3000 3000 3000>
    componentLinearAccScale=<0.3 0.3 0.3>
    componentLinearVelScale=<0.3 0.3 0.3>
    gravityOverride {z=-980}
    physicsBodyDefinitions [{
      boundBone {boneName=dyn_strap_R_04}
      boxExtents=<12 16 8>
      constraintSetup {angularLimitsMax=<0 10 0>, angularLimitsMin=<0 -40 0>, angularTarget=<0 0 0>}
      localJointOffset=<3 0 0>
     }]
   }
   n28 = Anim Dynamics - Bone: dyn_buckle_L!(ComponentPose=@n27, Alpha=1, GravityOverride:Vector="0,0,-700") -> (Pose) {
    angularDampingOverride=0.9999
    angularSpringConstant=1000
    bAngularSpring=true
    bGravityOverrideInSimSpace=true
    bOverrideAngularDamping=true
    simulationSpace=RootRelative
    alphaBoolBlend {blendOption=Linear}
    boundBone {boneName=dyn_buckle_L}
    componentAppliedLinearAccClamp=<5000 5000 5000>
    componentLinearAccScale=<0.5 0.5 0.5>
    componentLinearVelScale=<0.6 0.6 0.6>
    gravityOverride {z=-980}
    physicsBodyDefinitions [{
      sphereCollisionRadius=0
      boundBone {boneName=dyn_buckle_L}
      boxExtents=<5 5 5>
      constraintSetup {angularLimitsMax=<0 20 0>, angularLimitsMin=<0 -60 0>, angularTarget=<0 0 0>}
      localJointOffset=<1 0 0>
     }]
    relativeSpaceBone {boneName=head}
   }
   n29 = Anim Dynamics - Bone: dyn_loop_R_01(ComponentPose=@n28, Alpha=1, ExternalForce="0,0,0") {
    angularDampingOverride=0.9999
    angularSpringConstant=500
    bOverrideAngularDamping=true
    bUseGravityOverride=true
    gravityScale=-1
    simulationSpace=RootRelative
    alphaBoolBlend {blendOption=Linear}
    boundBone {boneName=dyn_loop_R_01}
    componentAppliedLinearAccClamp=<3000 3000 3000>
    componentLinearAccScale=<0.3 0.3 0.3>
    componentLinearVelScale=<0.3 0.3 0.3>
    gravityOverride {z=-800}
    physicsBodyDefinitions [{
      boundBone {boneName=dyn_loop_R_01}
      boxExtents=<8 12 6>
      constraintSetup {angularLimitsMax=<10 10 30>, angularLimitsMin=<-10 -50 -30>, angularTarget=<0 0 0>}
      localJointOffset=<1.2 0 0>
     }]
   }
   n30 = Anim Dynamics - Bone: dyn_loop_L_01(ComponentPose=@n29, Alpha=1, ExternalForce="0,0,0") {
    angularDampingOverride=0.9999
    angularSpringConstant=500
    bOverrideAngularDamping=true
    bUseGravityOverride=true
    gravityScale=-1
    simulationSpace=RootRelative
    alphaBoolBlend {blendOption=Linear}
    boundBone {boneName=dyn_loop_L_01}
    componentAppliedLinearAccClamp=<3000 3000 3000>
    componentLinearAccScale=<0.4 0.4 0.4>
    componentLinearVelScale=<0.3 0.3 0.3>
    gravityOverride {z=-800}
    physicsBodyDefinitions [{
      boundBone {boneName=dyn_loop_L_01}
      boxExtents=<8 12 6>
      constraintSetup {angularLimitsMax=<10 10 30>, angularLimitsMin=<-10 -60 -30>, angularTarget=<0 0 0>}
      localJointOffset=<1.2 0 0>
     }]
   }
   n31 = NOT Boolean(A=Get bRigidBodyEnabled())
   n32 = "Constraint - dyn_draw_string_R_MSTR (R)"!(ComponentPose=@n30, bAlphaBoolEnabled:bool=@n31, ConstraintWeights_0=1) -> (Pose) {
    alphaInputType=Bool
    alphaBoolBlend {blendOption=Linear}
    boneToModify {boneName=dyn_draw_string_R_MSTR}
    constraintSetup [{transformType=Rotation, targetBone {boneName=thigh_r}}]
    constraintWeights [1]
   }
   n33 = "Constraint - dyn_draw_string_L_MSTR (R)"!(ComponentPose=@n32, bAlphaBoolEnabled:bool=@n31, ConstraintWeights_0=1) -> (Pose) {
    alphaInputType=Bool
    alphaBoolBlend {blendOption=Linear}
    boneToModify {boneName=dyn_draw_string_L_MSTR}
    constraintSetup [{transformType=Rotation, targetBone {boneName=thigh_l}}]
    constraintWeights [1]
   }
   n34 = Anim Dynamics - Bone: dyn_draw_string_R_01(ComponentPose=@n33, Alpha=1, ExternalForce="0,0,-100") {
    angularDampingOverride=0.7
    angularSpringConstant=500
    bAngularSpring=true
    bOverrideAngularDamping=true
    bUseGravityOverride=true
    gravityScale=-1
    simulationSpace=RootRelative
    alphaBoolBlend {blendOption=Linear}
    boundBone {boneName=dyn_draw_string_R_01}
    componentAppliedLinearAccClamp=<3000 3000 3000>
    componentLinearAccScale=<0.3 0.3 0.3>
    componentLinearVelScale=<0.3 0.3 0.3>
    gravityOverride {z=-600}
    physicsBodyDefinitions [{
      boundBone {boneName=dyn_draw_string_R_01}
      boxExtents=<12 16 8>
      constraintSetup {angularLimitsMax=<5 -10 30>, angularLimitsMin=<-5 -50 -30>, angularTarget=<0 0 0>}
      localJointOffset=<3 0 0>
     }]
   }
   n35 = Anim Dynamics - Bone: dyn_draw_string_R_02(ComponentPose=@n34, Alpha=1, ExternalForce="0,0,-100") {
    angularDampingOverride=0.7
    angularSpringConstant=500
    bAngularSpring=true
    bOverrideAngularDamping=true
    bUseGravityOverride=true
    gravityScale=-1
    simulationSpace=RootRelative
    alphaBoolBlend {blendOption=Linear}
    boundBone {boneName=dyn_draw_string_R_02}
    componentAppliedLinearAccClamp=<3000 3000 3000>
    componentLinearAccScale=<0.3 0.3 0.3>
    componentLinearVelScale=<0.3 0.3 0.3>
    gravityOverride {z=-600}
    physicsBodyDefinitions [{
      boundBone {boneName=dyn_draw_string_R_02}
      boxExtents=<12 16 8>
      constraintSetup {angularLimitsMax=<5 10 20>, angularLimitsMin=<-5 -30 -20>, angularTarget=<0 0 0>}
      localJointOffset=<3 0 0>
     }]
   }
   n36 = Anim Dynamics - Bone: dyn_draw_string_R_02(ComponentPose=@n35, Alpha=1, ExternalForce="0,0,-100") {
    angularDampingOverride=0.7
    angularSpringConstant=500
    bAngularSpring=true
    bOverrideAngularDamping=true
    bUseGravityOverride=true
    gravityScale=-1
    simulationSpace=RootRelative
    alphaBoolBlend {blendOption=Linear}
    boundBone {boneName=dyn_draw_string_R_02}
    componentAppliedLinearAccClamp=<3000 3000 3000>
    componentLinearAccScale=<0.3 0.3 0.3>
    componentLinearVelScale=<0.3 0.3 0.3>
    gravityOverride {z=-600}
    physicsBodyDefinitions [{
      boundBone {boneName=dyn_draw_string_R_02}
      boxExtents=<12 16 8>
      constraintSetup {angularLimitsMax=<5 0 10>, angularLimitsMin=<-5 -30 -10>, angularTarget=<0 0 0>}
      localJointOffset=<3 0 0>
     }]
   }
   n37 = Anim Dynamics - Bone: dyn_draw_string_L_01(ComponentPose=@n36, Alpha=1, ExternalForce="0,0,-100") {
    angularDampingOverride=0.7
    angularSpringConstant=500
    bAngularSpring=true
    bOverrideAngularDamping=true
    bUseGravityOverride=true
    gravityScale=-1
    simulationSpace=RootRelative
    alphaBoolBlend {blendOption=Linear}
    boundBone {boneName=dyn_draw_string_L_01}
    componentAppliedLinearAccClamp=<3000 3000 3000>
    componentLinearAccScale=<0.3 0.3 0.3>
    componentLinearVelScale=<0.3 0.3 0.3>
    gravityOverride {z=-600}
    physicsBodyDefinitions [{
      boundBone {boneName=dyn_draw_string_L_01}
      boxExtents=<12 16 8>
      constraintSetup {angularLimitsMax=<5 10 10>, angularLimitsMin=<-5 -50 -10>, angularTarget=<0 0 0>}
      localJointOffset=<3 0 0>
     }]
   }
   n38 = Anim Dynamics - Bone: dyn_draw_string_L_02(ComponentPose=@n37, Alpha=1, ExternalForce="0,0,-100") {
    angularDampingOverride=0.7
    angularSpringConstant=500
    bAngularSpring=true
    bOverrideAngularDamping=true
    bUseGravityOverride=true
    gravityScale=-1
    simulationSpace=RootRelative
    alphaBoolBlend {blendOption=Linear}
    boundBone {boneName=dyn_draw_string_L_02}
    componentAppliedLinearAccClamp=<3000 3000 3000>
    componentLinearAccScale=<0.3 0.3 0.3>
    componentLinearVelScale=<0.3 0.3 0.3>
    gravityOverride {z=-600}
    physicsBodyDefinitions [{
      boundBone {boneName=dyn_draw_string_L_02}
      boxExtents=<12 16 8>
      constraintSetup {angularLimitsMax=<5 10 20>, angularLimitsMin=<-5 -30 -20>, angularTarget=<0 0 0>}
      localJointOffset=<3 0 0>
     }]
   }
   n39 = Anim Dynamics - Bone: dyn_draw_string_L_03(ComponentPose=@n38, Alpha=1, ExternalForce="0,0,-100") {
    angularDampingOverride=0.7
    angularSpringConstant=500
    bAngularSpring=true
    bOverrideAngularDamping=true
    bUseGravityOverride=true
    gravityScale=-1
    simulationSpace=RootRelative
    alphaBoolBlend {blendOption=Linear}
    boundBone {boneName=dyn_draw_string_L_03}
    componentAppliedLinearAccClamp=<3000 3000 3000>
    componentLinearAccScale=<0.3 0.3 0.3>
    componentLinearVelScale=<0.3 0.3 0.3>
    gravityOverride {z=-600}
    physicsBodyDefinitions [{
      boundBone {boneName=dyn_draw_string_L_03}
      boxExtents=<12 16 8>
      constraintSetup {angularLimitsMax=<5 10 10>, angularLimitsMin=<-5 -30 -10>, angularTarget=<0 0 0>}
      localJointOffset=<3 0 0>
     }]
    planarLimits [{drivingBone {boneName=pelvis}, planeTransform {location=<0 11 0>, rotation=<rot 0 -10 90>}}]
   }
   n40 = Anim Dynamics - Bone: dyn_draw_string_L_04(ComponentPose=@n39, Alpha=1, ExternalForce="0,0,-100") {
    angularDampingOverride=0.7
    angularSpringConstant=500
    bAngularSpring=true
    bOverrideAngularDamping=true
    bUseGravityOverride=true
    gravityScale=-1
    simulationSpace=RootRelative
    alphaBoolBlend {blendOption=Linear}
    boundBone {boneName=dyn_draw_string_L_04}
    componentAppliedLinearAccClamp=<3000 3000 3000>
    componentLinearAccScale=<0.3 0.3 0.3>
    componentLinearVelScale=<0.3 0.3 0.3>
    gravityOverride {z=-600}
    physicsBodyDefinitions [{
      boundBone {boneName=dyn_draw_string_L_04}
      boxExtents=<12 16 8>
      constraintSetup {angularLimitsMax=<5 0 10>, angularLimitsMin=<-5 -30 -10>, angularTarget=<0 0 0>}
      localJointOffset=<3 0 0>
     }]
   }
   Save cached pose 'AnimDynPose'(Pose=Component To Local(ComponentPose=@n40)) {cacheName=AnimDynPose}
  ]}
 EventGraph {nodes [
   Event OnCharacterCustomizationLoaded()
   > n1 = Set Skel Mesh(Skel Mesh=GetPartSkeletalMeshComponent())
   > n2 = Get Dynamic Material Instance(SkeletalMeshComponent=@n1)
   > Set Mat Instance Dynamic(Mat Instance Dynamic=@n2)
   > n3 = GetOrCreateDynamicOverlayMaterial(MeshComponent=@n1, MaterialName=None)
   > Set Overlay Mat Instance Dynamic(Overlay Mat Instance Dynamic=@n3)
   Event BlueprintInitializeAnimation()
   > Set Skel Mesh(Skel Mesh=GetOwnerSkeletalMeshComponent())
   n4 = TryGetPawnOwner()
   > Is Valid(InputObject=@n4)
   > n5 = Cast To FortPlayerPawnAthena(Object=@n4)
   > Set PlayerPawn(PlayerPawn:Fort Player Pawn Athena Object Reference=@n5) -> (Output_Get:Fort Player Pawn Athena Object Reference)
   n6 = Event BlueprintUpdateAnimation()
   > n7 = Sequence()
   > Is Valid(InputObject=Get PlayerPawn() -> (PlayerPawn:Fort Player Pawn Athena Object Reference))
   > Set WPO Globals Function(
    Material=Get Mat Instance Dynamic()
    Overlay Material=Get Overlay Mat Instance Dynamic()
    Skeletal Mesh=Get Skel Mesh()
    "Local Noise Transform  Location"="Get Local Noise Transform  Location"() -> ("Local Noise Transform  Location":Vector)
    Local Noise Offset=Get Local Noise Offset()
   )
   > Set WPO Regional Function!(
    execute
    Material:Material Instance Dynamic Object Reference=Get Mat Instance Dynamic()
    Overlay Material:Material Instance Dynamic Object Reference=Get Overlay Mat Instance Dynamic()
    Skeletal Mesh:Skeletal Mesh Component Object Reference=Get Skel Mesh()
    Velocity Transform:Name=foot_r
    Velocity Interp Speed:double=Get Velocity Interp Speed()
    Collision Transform:Name=calf_r
    Collision Transform Axis:EAxis Enum
    Invert Collision Transform Axis:bool=false
    Delta Time X:double=@n6.DeltaTimeX
    Previous Velocity Transform Location:"Vector (by ref)"=Get R1 Velocity Transform Location()
    Previous Velocity Transform Velocity:"Vector (by ref)"=Get R1 Velocity Transform Velocity()
    Region:int=1
   ) -> (then)
   > Set WPO Regional Function(
    Material=Get Mat Instance Dynamic()
    Overlay Material=Get Overlay Mat Instance Dynamic()
    Skeletal Mesh=Get Skel Mesh()
    Velocity Transform=foot_l
    Velocity Interp Speed=Get Velocity Interp Speed()
    Collision Transform=calf_l
    Invert Collision Transform Axis=true
    Delta Time X=@n6.DeltaTimeX
    Previous Velocity Transform Location=Get R2 Velocity Transform Location()
    Previous Velocity Transform Velocity=Get R2 Velocity Transform Velocity()
    Region=2
   )
   > n8 = Branch(Condition=Get bIsSkydiveFloating())
   > n9 = Set WPO Pho Wind Intensity(WPO Pho Wind Intensity=1)
   SetScalarParameterValue(execute=[@n10.then, @n9.then], self=Get Mat Instance Dynamic(), ParameterName=ClothSim_WindStrength, Value=Get WPO Pho Wind Intensity())
   n10 = Set WPO Pho Wind Intensity(execute=@n8.else, WPO Pho Wind Intensity=0)
   n11 = Get Skel Mesh()
   n12 = Get bIsMoving2D()
   Set AlphaThighR(
    execute=@n7.then_1
    AlphaThighR:double=float * float(
     A=GetAlphaFromReferenceTransformDelta(SkeletalMeshComponent=@n11, BoneName=thigh_r, BaseName=pelvis, InputRange=(X=-30,Y=-60), TwistAxis="1,0,0")
     B:bool=NOT Boolean(A=@n12)
    ) {is=title K2Node_PromotableOperator}
   ) -> (Output_Get:double)
  ]}
]
variables [
 AlphaThighL {type=number}
 AlphaThighR {type=number}
 IsCrouchMoving {type=boolean}
 Local Noise Offset {type=Vector}
 "Local Noise Transform  Location" {type=Vector}
 Mat Instance Dynamic {type=MaterialInstanceDynamic}
 Overlay Mat Instance Dynamic {type=MaterialInstanceDynamic}
 PlayerPawn {type=FortPlayerPawnAthena}
 R1 Velocity Transform Location {type=Vector}
 R1 Velocity Transform Velocity {type=Vector}
 R2 Velocity Transform Location {type=Vector}
 R2 Velocity Transform Velocity {type=Vector}
 Skel Mesh {type=SkeletalMeshComponent}
 Velocity Interp Speed {type=number, default=1}
 WPO Pho Wind Intensity {type=number}
]
